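; ToyFox knockback ability that launches actors or vehicles overlapped by a moving giant and applies damage to its avatar.
; It extends a parent knockback ability, using the granting effect's causer as launcher.
; Impulse direction opposes target velocity, or points away from the launcher for near-stationary targets.
; Hotfixable angle and speed ranges rotate that direction sideways so things ahead of the giant get pushed aside, with debug arrows drawn.
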(event EventDoLaunch
  (params (Vel :type Vector :default "0, 0, 0"))
  ;[0.03518 0 0.15 0.5] DoLaunch: When applying launch, also apply damage.
  (|CallParentFunction Vel)
  (Fortnite|ApplyGameplayEffectsToActor (Ability|GetAvatarActorfromActorInfo) (Utilities|Array|MakeArray (Variables|Default|GetOverlapKnockbackDamageGameplayEffect)) (0.0 0.0 0.0) false 1 (Ability|GetFortAbilitySystemComponent (Default|GetAbilityInstigator)) (Ability|GetGrantedbyEffectContext)))

(event EventLaunchVehicle
  (params (VehicleInterface :type Fort%20Vehicle%20Interface%20Interface) (LaunchVelocity :type Vector%20%28by%20ref%29 :default "0, 0, 0"))
  ;[0.03518 0 0.15 0.5] DoLaunch: When applying launch, also apply damage.
  (|CallParentFunction VehicleInterface LaunchVelocity)
  (Fortnite|ApplyGameplayEffectsToActor (Ability|GetAvatarActorfromActorInfo) (Utilities|Array|MakeArray (Variables|Default|GetOverlapKnockbackDamageGameplayEffect)) (0.0 0.0 0.0) false 1 (Ability|GetFortAbilitySystemComponent (Default|GetAbilityInstigator)) (Ability|GetGrantedbyEffectContext)))

(fn FindLauncher
  (params (DefaultLauncher :type Actor%20Object%20Reference))
  (return (Ability|EffectContext|GetEffectCauser (Ability|GetGrantedbyEffectContext))))

(fn GetInitialKnockMag
  (params (Direction :type Vector :default "0, 0, 0") (OverrideXYMag :type Boolean) (XYMAgOverride :type Float%20%28double-precision%29))
  (bind _launcher (Variables|KnockbackParent|GetLauncher))
  (bind _returnvalue (Transformation|GetActorLocation _launcher))
  (LocalVariables|SetCurrentGiantVelocity (Transformation|GetVelocity (Ability|EffectContext|GetEffectCauser (Ability|GetGrantedbyEffectContext))))
  (Utilities|FlowControl|Sequence
    (:then_0
      ; Calculate a proper impulse direction
      (LocalVariables|SetTargetVelocity (Transformation|GetVelocity (Variables|KnockbackParent|GetTarget)))
      ;; Target not really moving, use location difference to impulse
      (if (<= (Math|Vector|VectorLength (LocalVariables|GetTargetVelocity)) (ScalableFloat|GetCachedScalableFloatValue (Variables|Hotfix|GetHF_TargetNotMovingSpeed)))
        (LocalVariables|SetDirection2D (Math|Vector|Normalize2D(Vector) (- (Transformation|GetActorLocation (Variables|KnockbackParent|GetTarget)) (Transformation|GetActorLocation (Variables|KnockbackParent|GetLauncher)))))
        (else
          (LocalVariables|SetDirection2D (Math|Vector|NegateVector (Math|Vector|Normalize2D(Vector) (LocalVariables|GetTargetVelocity)))))))
    (:then_1
      ; Modify Angle Direction
      (LocalVariables|SetAngleModifyDirection (Math|Float|SelectFloat 1.0 -1.0 (> (Math|Vector|DotProduct (Math|Vector|GetRightVector (Math|Rotator|MakeRotfromXZ (LocalVariables|GetCurrentGiantVelocity) (Math|Vector|VectorUp))) (LocalVariables|GetDirection2D)) 0))))
    (:then_2
      ; Calculte modified impulse direction. So things in front of giant moving facing, will be pushed to sides
      (bind _output_get (LocalVariables|SetModifiedDirection (Math|Vector|RotateVectorAroundAxis (LocalVariables|GetDirection2D) (Utilities|Operators|Multiply (Math|Float|MapRangeClamped (Math|Vector|DotProduct (LocalVariables|GetDirection2D) (Math|Vector|Normalize (LocalVariables|GetCurrentGiantVelocity) 1e+01)) (Math|Trig|Cos(Degrees) (ScalableFloat|GetCachedScalableFloatValue (Variables|Hotfix|GetHF_AngleToFullyModifyImpulse))) (Math|Trig|Cos(Degrees) (ScalableFloat|GetCachedScalableFloatValue (Variables|Hotfix|GetHF_AngleToLerpOutModifyImpulseAngle))) (ScalableFloat|GetCachedScalableFloatValue (Variables|Hotfix|GetHF_MaxModifyImpulseAngle))) (LocalVariables|GetAngleModifyDirection) (Math|Float|MapRangeClamped (Math|Vector|VectorLength (LocalVariables|GetCurrentGiantVelocity)) (ScalableFloat|GetCachedScalableFloatValue (Variables|Hotfix|GetHF_MinGiantSpeedToModifyImpulse)) (ScalableFloat|GetCachedScalableFloatValue (Variables|Hotfix|GetHF_MaxGiantSpeedToModifyImpulse)) 0.0 1.0)) "0.000000,0.000000,1.000000")))
      ;[0.15 0.1267 0 0.5] Draw Debug
      (Rendering|Debug|DrawDebugArrow _returnvalue (+ _returnvalue (* (LocalVariables|GetDirection2D) 2e+03)) 1e+01 "(R=1.000000,G=0.876913,B=0.888391,A=1.000000)" 5.0)
      (Rendering|Debug|DrawDebugArrow _returnvalue (+ _returnvalue (* (LocalVariables|GetModifiedDirection) 2e+03)) 2e+01 "(R=1.000000,G=0.000000,B=0.775500,A=1.000000)" 5.0)
      (return (|CallParentFunction _output_get)))))
